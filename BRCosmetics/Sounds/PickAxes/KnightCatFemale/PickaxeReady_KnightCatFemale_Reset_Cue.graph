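# Sound cue for the KnightCatFemale pickaxe ready reset, playing one of six "Settle" sound variations.
# An equal-weight random node picks the wave, then a modulator adds slight pitch and volume variation.
# The result passes through close-range pickaxe attenuation into a root mixer whose second input is empty.
# The cue uses a reduced volume, the melee weapon sound class and the shared cosmetic pickaxe-ready property sheet.
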SoundCue /BRCosmetics/Sounds/PickAxes/KnightCatFemale/PickaxeReady_KnightCatFemale_Reset_Cue
root=Mixer_0
nodes [
 {id=Attenuation_0, children [Modulator_0], properties {attenuationSettings=/Game/Sounds/Attenuation/Fort_Weapon_Pickaxe_Close_Att}}
 {id=Mixer_0, children [Attenuation_0, null], properties {inputVolume [1,1]}}
 {id=Modulator_0, children [Random_0], properties {pitchMax=1.01, pitchMin=0.99, volumeMax=1.02, volumeMin=0.98}}
 {
  id=Random_0
  children [WavePlayer_13, WavePlayer_14, WavePlayer_15, WavePlayer_16, WavePlayer_17, WavePlayer_18]
  properties {weights [1,1,1,1,1,1]}
 }
 {id=WavePlayer_13, properties {soundWaveAssetPtr=&/Pickaxe_KnightCatFemale_Settle_05}}
 {id=WavePlayer_14, properties {soundWaveAssetPtr=&/Pickaxe_KnightCatFemale_Settle_06}}
 {id=WavePlayer_15, properties {soundWaveAssetPtr=&/Pickaxe_KnightCatFemale_Settle_03}}
 {id=WavePlayer_16, properties {soundWaveAssetPtr=&/Pickaxe_KnightCatFemale_Settle_04}}
 {id=WavePlayer_17, properties {soundWaveAssetPtr=&/Pickaxe_KnightCatFemale_Settle_01}}
 {id=WavePlayer_18, properties {soundWaveAssetPtr=&/Pickaxe_KnightCatFemale_Settle_02}}
]
properties {
 audioPropertiesSheet=/BRCosmetics/Sounds/Settings/PropertySheets/Pickaxe/APS_Cosmetic_Pickaxe_Ready
 maxDistance=1536
 soundClassObject=&Fort_WeaponMelee_PSM
 volumeMultiplier=0.4
 concurrencySet [null]
}
